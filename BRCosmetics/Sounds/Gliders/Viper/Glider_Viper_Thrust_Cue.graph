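# Sound cue for the Viper glider's thrust-start sound, a cosmetic glider effect.
# A random node picks with equal weight among four Thrust_Start wave variations to vary each playback.
# Playback uses shared glider thrust attenuation and concurrency, a glider sound class, a cosmetic glider loop audio properties sheet, and a reduced cue volume.
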SoundCue /BRCosmetics/Sounds/Gliders/Viper/Glider_Viper_Thrust_Cue
root=Random_0
nodes [
 {id=Random_0, children [WavePlayer_0,WavePlayer_1,WavePlayer_5,WavePlayer_6], properties {weights [1,1,1,1]}}
 {id=WavePlayer_0, properties {soundWaveAssetPtr=&/Glider_Viper_Thrust_Start_01}}
 {id=WavePlayer_1, properties {soundWaveAssetPtr=&/Glider_Viper_Thrust_Start_04}}
 {id=WavePlayer_5, properties {soundWaveAssetPtr=&/Glider_Viper_Thrust_Start_03}}
 {id=WavePlayer_6, properties {soundWaveAssetPtr=&/Glider_Viper_Thrust_Start_02}}
]
properties {
 attenuationSettings=&Fort_Player_Glider_Thrust_Athena_Att
 audioPropertiesSheet=&APS_Cosmetic_Glider_Loop
 maxDistance=12500
 soundClassObject=&Fort_Gliders_Land
 volumeMultiplier=0.4
 concurrencySet [&Athena_Glider_Thrust_Conc]
}
